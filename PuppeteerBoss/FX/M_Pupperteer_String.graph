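# Translucent unlit Niagara mesh-particle material for the Puppeteer boss's strings, two-sided and colored by particle color.
# World position offset combines SplineThicken width with a time-scrolled sine wobble, randomized per particle and masked along the string's V gradient.
# Opacity builds from UV sine gradients eroded by blurred noise, then cut by StringGrow and softened with DepthFade.
# Dynamic parameters let particles drive time offset, growth, width, wobble intensity and erosion.
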Material /PuppeteerBoss/FX/M_Pupperteer_String
graph {
 properties {
  bAutomaticallySetUsageInEditor=false
  bUsedWithNiagaraMeshParticles=true
  blendMode=BLEND_Translucent
  shadingModel=MSM_Unlit
  twoSided=true
  usedShadingModels=Unlit
 }
 nodes [
  n1 = Multiply(
   A=ScalarParameter() {defaultValue=2.5, parameterName=Width}
   B=DynamicParameter() {defaultValue=<rgba 0 0 1 1>, paramNames [TimeOffset, StringGrow, StringWidth, WPO_Intensity]}.StringWidth
  )
  n2 = MaterialFunctionCall() {materialFunction=&LinearGradient}
  n3 = Multiply(A=Sine(None=Multiply(A=@n2.VGradient) {constB=0.5}), B=Constant() {r=1}) {constB=5}
  n4 = WorldPosition()
  n5 = DynamicParameter() {defaultValue=<rgba 0 0 1 1>, paramNames [TimeOffset, StringGrow, StringWidth, WPO_Intensity]}
  n6 = Multiply(A=Add(A=Time(), B=Multiply(A=@n5.TimeOffset) {constB=100}), B=ScalarParameter() {defaultValue=400, parameterName=WPO_Speed})
  n7 = ScalarParameter() {defaultValue=1000, parameterName=WPO_Tile}
  n8 = LinearInterpolate(Alpha=ParticleRandom()) {constA=0.5}
  n9 = Sine(None=Divide(A=Add(A=@n4.XYZ, B=@n6), B=Multiply(A=@n7, B=@n8)) {constB=4000})
  n10 = Multiply(A=ScalarParameter() {defaultValue=5, parameterName=WPO_Intensity}, B=@n5.WPO_Intensity)
  Add(
   A=MaterialFunctionCall(WidthBase=@n1, WidthTip=@n1) {materialFunction=&SplineThicken}.WorldPositionOffset
   B=Multiply(A=Saturate(None=@n3), B=Multiply(A=@n9, B=@n10))
  ) {constB=0}
  n11 = ParticleColor()
  n12 = MaterialFunctionCall() {materialFunction=&LinearGradient}
  n13 = Multiply(A=@n12.UGradient) {constB=0.5}
  n14 = DynamicParameter() {parameterIndex=1, defaultValue=<rgba 0 0 1 1>, paramNames [Erode,Param2,Param3,Param4]}
  n15 = Multiply(A=TextureCoordinate(), B=Constant2Vector() {g=3.5, r=1})
  n16 = Subtract(
   A=Saturate(None=Sine(None=@n13))
   B=Multiply(
    A=LinearInterpolate(Alpha=@n14.Erode) {constB=5}
    B=TextureSample(UVs=@n15) {samplerType=SAMPLERTYPE_Grayscale, texture=&LowResBlurredNoise}.R
   )
  )
  n17 = Sine(None=Multiply(A=@n12.VGradient) {constB=0.5})
  n18 = Multiply(A=Saturate(None=Multiply(A=@n17) {constB=20}), B=LinearInterpolate(Alpha=@n12.VGradient))
  n19 = Saturate(None=Subtract(
    A=Multiply(A=@n16, B=@n18)
    B=DynamicParameter() {defaultValue=<rgba 0 0 1 1>, paramNames [TimeOffset, StringGrow, StringWidth, WPO_Intensity]}.StringGrow
   ))
  DepthFade(Opacity=Multiply(A=Multiply(A=@n11.A, B=@n19), B=ScalarParameter() {defaultValue=0.5, parameterName=OverallOpacity})) {fadeDistanceDefault=20}
  MaterialFunctionCall(Emissive=Multiply(A=ScalarParameter() {defaultValue=1, parameterName=Emissive}, B=@n11.RGB)) {materialFunction=&MF_FX_EmissiveAdjustments}
  MaterialFunctionCall() {materialFunction=&DebugTimeSine}
 ]
}
