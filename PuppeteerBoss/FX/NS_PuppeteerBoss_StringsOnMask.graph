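# Niagara system for the Puppeteer boss's mask: puppet strings, glowing eye weakpoints and idle smoke.
# Ten mesh strings stretch from sampled skeletal-mesh points to two user targets, fade when overlong, and switch material animation on IsStringConnecting.
# Decal, wave and wash sprites mark those anchors; MaskDBNO suppresses decals, waves and a smoke material parameter.
# Socket-attached eye flashes with lights and GPU sparkles follow weakpoint and per-eye hide flags, while a system Timeline boosts the flash.
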NiagaraSystem /PuppeteerBoss/FX/NS_PuppeteerBoss_StringsOnMask
dependencies {
 usedDataInterfaces [
  &NiagaraDataInterfaceArrayFloat
  &NiagaraDataInterfaceArrayFloat3
  &NiagaraDataInterfaceCurve
  &NiagaraDataInterfaceColorCurve
  &NiagaraDataInterfaceSkeletalMesh
 ]
 usedDynamicInputs []
 usedModules [
  &self:SystemSpawnScript.SystemScriptSource.SystemScriptGraph.NiagaraNodeAssignment_1.SetVariables_#1
  &SystemState
  &Timeline
 ]
}
emitters {
 EyeFlash {
  emitterUpdateScript [
   &EmitterState(Life Cycle Mode=::NewEnumerator1, Loop Behavior=::NewEnumerator1, Loop Duration Mode=::NewEnumerator1, UseLoopDelay=false)
   &SpawnBurst_Instantaneous(Spawn Count=1, Use Spawn Probability=false)
   &SpawnBurst_Instantaneous(Spawn Count=1, Spawn Group=1, Use Spawn Probability=false) {module=SpawnBurst_Instantaneous001}
  ]
  particleSpawnScript [
   &InitializeParticle(
    Color=<rgba 1 0.0967 0.0867 1>
    Sprite Rotation Mode=::NewEnumerator2
    Sprite Size Mode=::NewEnumerator3
    Uniform Sprite Size=300
    UsePositionOffset=false
    Write Material Random=true
   )
   &DynamicMaterialParameters(
    Index 0 Param 1=&RandomRangeFloat(Minimum=0.4, Maximum=0.6)
    Index 0 Param 2=1
    Index 0 Param 3=1
    Index 0 Param 4=1
    Module.Param0WriteEnabled=true
    Module.Param1WriteEnabled=true
    Module.Param2WriteEnabled=true
    Module.Param3WriteEnabled=true
   )
  ]
  particleUpdateScript [
   &ParticleState(Kill Particles When Lifetime Has Elapsed=false, Let Infinitely Lived Particles Die When Emitter Deactivates=true)
   &SkeletalMeshLocation(
    Bone / Socket Index=&SpawnGroup()
    MaskSpawnGroup=false
    Mesh Sampling Type=::NewEnumerator5
    Skeletal Mesh=$System.SK_Mesh
    Socket Sampling Mode=::NewEnumerator5
   )
   &CameraOffset(Camera Offset Amount=&Lerp_Float(A=25, B=50, Alpha=$System.Timeline.Float)) {module=CameraOffset001}
   &ScaleSpriteSize(
    Scale Sprite Size Mode=::NewEnumerator0
    Uniform Scale Factor=&Multiply_Float(
     A=&FloatFromCurve(
      FloatCurve=DataInterface() {propertyValues {Curve {keys [
          {arriveTangent=13.3324, interpMode=RCIM_Cubic, leaveTangent=13.3324, tangentMode=RCTM_User, time=0, value=0}
          {arriveTangent=1, interpMode=RCIM_Cubic, leaveTangent=1, time=0.2, value=2}
          <at 1 1 RCIM_Cubic>
         ]}}}
      CurveIndex=$Particles.Age
     )
     B=&Lerp_Float(
      A=0.7
      B=1
      Alpha=&Waveform("Waveform [1]"=::NewEnumerator0, "[1] Phase Scale"=2, "[1] Frequency"=1, "[1] Amplitude Min/Max"=<0 1>)
     )
    )
   )
   &ScaleSpriteSize(
    Scale Sprite Size Mode=::NewEnumerator0
    Uniform Scale Factor=&SelectFloatFromArray(
     Direct Array Index=&Add_Integer(
      A=&SpawnGroup()
      B=&self:HideEyeGlowReturnIndex(HideLeft:NiagaraBool=$User.HideLeftEyeFX, HideRight:NiagaraBool=$User.HideRightEyeFX)
     )
     Float Selection Array=DataInterface() {propertyValues {FloatData [0,1,1,0]}}
    )
   ) {module=ScaleSpriteSize001}
   &ScaleSpriteSize(
    Scale Sprite Size Mode=::NewEnumerator0
    Uniform Scale Factor=&Lerp_Float(
     A=1
     B=&SelectFloatFromArray(
      Direct Array Index=&Add_Integer(
       A=&SpawnGroup()
       B=&self:FlashEyeGlowReturnIndex(IsLeft:NiagaraBool=$User.IsLeftEye)
      )
      Float Selection Array=DataInterface() {propertyValues {FloatData [1,3,3,1]}}
     )
     Alpha=$System.Timeline.Float
    )
   ) {module=ScaleSpriteSize002}
   &ScaleColor(
    Scale Alpha=&MakeCustomFloatFromBool(Bool=$User.EnableWeakpoints, True Float=1, False Float=0)
    Scale RGB=&VectorFromFloat(Value=&Multiply_Float(
      A=&MakeCustomFloatFromBool(Bool=$User.IsLeftEye, True Float=&Lerp_Float(A=1, B=2, Alpha=$System.Timeline.Float), False Float=1)
      B=3
     ))
    ScaleA=true
    ScaleRGB=true
   )
   &FN_Light_Attributes(
    Color=$Particles.Color
    Create Color Attribute=true
    Create Radius Attribute=true
    Radius=&Multiply_Float(
     A=&Multiply_Float(
      A=&Multiply_Float(
       A=0.25
       B=&FloatFromCurve(
        FloatCurve=DataInterface() {propertyValues {Curve {keys [
            {arriveTangent=13.3324, interpMode=RCIM_Cubic, leaveTangent=13.3324, tangentMode=RCTM_User, time=0, value=0}
            {arriveTangent=1, interpMode=RCIM_Cubic, leaveTangent=1, time=0.2, value=2}
            <at 1 1 RCIM_Cubic>
           ]}}}
        CurveIndex=$Particles.Age
       )
      )
      B=&MakeCustomFloatFromBool(Bool=$User.EnableWeakpoints, True Float=1, False Float=0)
     )
     B=&MakeFloatFromVector2d(Vector 2D=$Particles.SpriteSize)
    )
    Scale Color=5
   )
  ]
  rendererClasses [&NiagaraSpriteRendererProperties, &NiagaraLightRendererProperties]
  renderers [
   {rendererClass=&NiagaraSpriteRendererProperties}
   {rendererClass=&NiagaraLightRendererProperties, rendererIndex=1}
  ]
 }
 EyeSparkles {
  simTarget=GPUComputeSim
  emitterUpdateScript [
   &EmitterState()
   &SpawnRate(
    Spawn Probability=&MakeCustomFloatFromBool(
     Bool=$User.HideLeftEyeFX
     True Float=0
     False Float=&MakeCustomFloatFromBool(Bool=$User.EnableWeakpoints, True Float=1, False Float=0)
    )
    SpawnRate=50
    Use Spawn Probability=true
   )
   &SpawnRate(
    Spawn Probability=&MakeCustomFloatFromBool(
     Bool=$User.HideRightEyeFX
     True Float=0
     False Float=&MakeCustomFloatFromBool(Bool=$User.EnableWeakpoints, True Float=1, False Float=0)
    )
    SpawnGroup=1
    SpawnRate=100
    Use Spawn Probability=true
   ) {module=SpawnRate001}
  ]
  particleSpawnScript [
   &InitializeParticle(
    Color=<rgba 1 0.072 0.0554 1>
    Lifetime Max=3
    Lifetime Min=2
    Lifetime Mode=::NewEnumerator1
    Sprite Rotation Mode=::NewEnumerator2
    Sprite Size Mode=::NewEnumerator1
    Uniform Sprite Size Max=5
    Uniform Sprite Size Min=2
    UsePositionOffset=false
    Write Material Random=true
   )
   &DynamicMaterialParameters(
    Index 0 Param 1=&RandomRangeFloat(Minimum=0.4, Maximum=0.6)
    Index 0 Param 2=1
    Index 0 Param 3=1
    Index 0 Param 4=1
    Module.Param0WriteEnabled=true
    Module.Param1WriteEnabled=true
    Module.Param2WriteEnabled=true
    Module.Param3WriteEnabled=true
   )
   &SkeletalMeshLocation(
    Bone / Socket Index=$Emitter.SpawnGroup
    MaskSpawnGroup=false
    Mesh Sampling Type=::NewEnumerator5
    Skeletal Mesh=$System.SK_Mesh
    Socket Sampling Mode=::NewEnumerator5
   )
   &CameraOffset(Camera Offset Amount=50) {module=CameraOffset001}
   &ShapeLocation(
    MaskSpawnGroup=false
    Override Seed=false
    Shape Origin=$Particles.SkeletalMeshLocation.SampledPosition
    Sphere Radius=10
    UseOwnerScale=true
   )
   &AddVelocity(
    Module.Use Falloff Distance=false
    Override Seed=false
    Velocity Mode=::NewEnumerator1
    Velocity Origin=$Particles.SkeletalMeshLocation.SampledPosition
    Velocity Speed=&RandomRangeFloat(Minimum=1, Maximum=1)
   )
  ]
  particleUpdateScript [
   &ParticleState()
   &CurlNoiseForce(
    Bias Noise=false
    Noise Frequency=100
    Noise Strength=&Multiply_Float(A=&RandomRangeFloat(Minimum=0.8, Maximum=1), B=500)
    Pan Noise=true
    Pan Noise Field=<0.5 0.5 0.5>
   )
   &InheritSourceMovement(
    Applied Position Scale=&Multiply_VectorByFloat(
     Vector=<1 1 0.3>
     Float=&FloatFromCurve(FloatCurve=DataInterface() {propertyValues {Curve {keys [
          {arriveTangent=-2.3182, interpMode=RCIM_Cubic, leaveTangent=-2.3182, tangentMode=RCTM_User, time=0, value=1}
          {interpMode=RCIM_Cubic, tangentMode=RCTM_User, time=1, value=0}
         ]}}})
    )
    ApplyForce=false
    ApplyPosition=true
    ApplyPositionFalloff=false
    Source Velocity:Vector3f=&Add_Vector(A=$Particles.SkeletalMeshLocation.SampledVelocity, B=$Engine.Owner.Velocity)
   )
   &Drag(Use Linear Drag=true, Use Rotational Drag=true)
   &AccelerationForce(Acceleration=&MakeVector(
     X=0
     Y=0
     Z=&Multiply_Float(A=100, B=&RandomRangeFloat(Minimum=0.5, Maximum=1))
    ))
   &SolveForcesAndVelocity(Clamp Velocity=false, Limit Acceleration=false)
   &ScaleSpriteSize(Uniform Curve Sprite Scale=DataInterface() {propertyValues {Curve {keys [
        {interpMode=RCIM_Cubic, tangentMode=RCTM_User, time=0, value=0}
        <at 0.1 1 RCIM_Cubic>
        {arriveTangent=-3.2677, interpMode=RCIM_Cubic, leaveTangent=-3.2677, tangentMode=RCTM_User, time=1, value=0}
       ]}}}) {module=ScaleSpriteSize003}
   &ScaleColor(
    Scale Alpha=&RandomRangeFloat(Minimum=0.3, Maximum=1)
    Scale RGB=&VectorFromFloat(Value=&Lerp_Float(
      A=5
      B=100
      Alpha=&FloatFromCurve(FloatCurve=DataInterface() {propertyValues {Curve {keys [
           {arriveTangent=9.0725, interpMode=RCIM_Cubic, leaveTangent=9.0725, tangentMode=RCTM_User, time=0, value=0}
           {arriveTangent=-2.5249, interpMode=RCIM_Cubic, leaveTangent=-2.5249, tangentMode=RCTM_User, time=0.1408, value=0.9847}
           {arriveTangent=-0.0405, interpMode=RCIM_Cubic, leaveTangent=-0.0405, tangentMode=RCTM_User, time=1, value=0}
          ]}}})
     ))
    ScaleA=true
    ScaleRGB=true
   ) {module=ScaleColor001}
  ]
  rendererClasses [&NiagaraSpriteRendererProperties]
  renderers [{rendererClass=&NiagaraSpriteRendererProperties}]
 }
 IdleSmoke {
  emitterUpdateScript [
   &EmitterState(Life Cycle Mode=::NewEnumerator1, Loop Duration=1, UseLoopDelay=true)
   /SetVariables(Emitter.LightingCenter:Vector4f=&MakeFloat4FromVectorAndFloat(
     "Vector (RGB)"=&AddVectorToPosition(Position=$Engine.Emitter.SimulationPosition, Vector=<0 0 -50>)
     "Float (Alpha)"=1
    )) {isSetParameters=true}
   &SpawnRate(SpawnRate=8, Use Spawn Probability=false)
  ]
  particleSpawnScript [
   &PartitionParticles(Number Of Partitions=3)
   &InitializeParticle(
    Color=<rgba 1 0.8071 0.7872 1>
    Lifetime=&RandomRangeFloat(Minimum=1.4, Maximum=2)
    Position Offset=<0 0 270>
    Sprite Rotation Mode=::NewEnumerator2
    Sprite Size Mode=::NewEnumerator3
    Uniform Sprite Size=&RandomRangeFloat(Minimum=112.8685, Maximum=144.8331)
    UsePositionOffset=true
    Write Material Random=true
   ) {module=InitializeParticle001}
   &CameraOffset(Camera Offset Amount=&Multiply_Float(
     A=&RandomRangeFloat(Minimum=25, Maximum=50)
     B=&SelectFloatFromArray(Direct Array Index=$Particles.PartitionParticles.Partition, Float Selection Array=DataInterface() {propertyValues {FloatData [0,0,0]}})
    ))
  ]
  particleUpdateScript [
   &ParticleState()
   &ScaleSpriteSize(
    Scale Sprite Size Mode=::NewEnumerator0
    Uniform Scale Factor=&SelectFloatFromArray(Direct Array Index=$Particles.PartitionParticles.Partition, Float Selection Array=DataInterface() {propertyValues {FloatData [1.3,0.5,0.5]}})
   ) {module=ScaleSpriteSize001}
   &ScaleSpriteSize(Scale Factor=&Vector2DFromFloat(Value=&FloatFromCurve(
      FloatCurve=DataInterface() {propertyValues {Curve {keys [{interpMode=RCIM_Cubic, leaveTangent=1.7304, tangentMode=RCTM_Break, time=0, value=0.5}, <at 0.8 1>]}}}
      Scale Curve=4
     )))
   &SkeletalMeshLocation(
    Bone / Socket Index=&Add_Integer(A=$Particles.PartitionParticles.Partition, B=12)
    Bone Sampling Mode=::NewEnumerator4
    MaskSpawnGroup=false
    Orientation Sampling=::NewEnumerator1
    Sampled Position Offset=&SelectVectorFromArray(
     Direct Array Index=$Particles.PartitionParticles.Partition
     Vector Selection Array=DataInterface() {propertyValues {FloatData [<-50 0 0>, <-6 0 0>, <6 0 0>]}}
    )
    Skeletal Mesh=$System.SK_Mesh
   )
   &SolveForcesAndVelocity(Clamp Velocity=false, Limit Acceleration=false)
   /SetVariables(Particles.Color_ParticleEnergy:LinearColor=&ColorFromCurve(ColorCurve=DataInterface() {propertyValues {
       AlphaCurve {keys [<at 0 1 RCIM_Cubic>]}
       BlueCurve {keys [<at 0.3364 0.3332>,<at 0.749 0.0641 RCIM_Cubic>]}
       GreenCurve {keys [<at 0.3364 0.3531>,<at 0.749 0 RCIM_Cubic>]}
       RedCurve {keys [<at 0.3364 2>,<at 0.749 5 RCIM_Cubic>]}
      }})) {isSetParameters=true}
   &DynamicMaterialParameters(
    Index 0 Param 1=&FloatFromCurve(FloatCurve=DataInterface() {propertyValues {Curve {keys [{tangentMode=RCTM_User, time=0, value=0}, {tangentMode=RCTM_User, time=1, value=1}]}}})
    Index 0 Param 3=&FloatFromCurve(FloatCurve=DataInterface() {propertyValues {Curve {keys [
         {interpMode=RCIM_Cubic, tangentMode=RCTM_Break, time=0, value=0}
         {interpMode=RCIM_Cubic, leaveTangent=-3.8707, tangentMode=RCTM_Break, time=0.2, value=1}
         {tangentMode=RCTM_User, time=1, value=0}
        ]}}})
    Index 0 Param 4=&MakeCustomFloatFromBool(Bool=$User.MaskDBNO, True Float=0, False Float=1)
    Index 1 Param 1=&MakeFloatFromVector(Vector=&MakeVectorFromLinearColorRGB(Color=$Particles.Color_ParticleEnergy), Channel=::NewEnumerator0)
    Index 1 Param 2=&MakeFloatFromVector(Vector=&MakeVectorFromLinearColorRGB(Color=$Particles.Color_ParticleEnergy), Channel=::NewEnumerator1)
    Index 1 Param 3=&MakeFloatFromVector(Vector=&MakeVectorFromLinearColorRGB(Color=$Particles.Color_ParticleEnergy), Channel=::NewEnumerator2)
    Index 1 Param 4=&FloatFromCurve(FloatCurve=DataInterface() {propertyValues {Curve {keys [{interpMode=RCIM_Cubic, tangentMode=RCTM_Break, time=0.4, value=0}, {arriveTangent=3.131, tangentMode=RCTM_User, time=1, value=1}]}}})
    Index 2 Param 1=500
    Index 2 Param 2=200
    Index 2 Param 3=500
    Module.Index0Param0Write=true
    Module.Index0Param1Write=true
    Module.Index0Param2Write=true
    Module.Index0Param3Write=true
    Module.Index1Param0Write=true
    Module.Index1Param1Write=true
    Module.Index1Param2Write=true
    Module.Index1Param3Write=true
    Module.Index2Param0Write=true
    Module.Index2Param1Write=true
    Module.Index2Param2Write=true
    Module.Index2Param3Write=true
    Write Parameter Index 1=true
    Write Parameter Index 2=true
   )
   &ScaleColor(
    Scale Alpha=&FloatFromCurve(FloatCurve=DataInterface() {propertyValues {Curve {keys [
         {interpMode=RCIM_Cubic, leaveTangent=7.3008, tangentMode=RCTM_Break, time=0, value=0}
         {interpMode=RCIM_Cubic, tangentMode=RCTM_Break, time=0.2, value=1}
         {interpMode=RCIM_Cubic, leaveTangent=-0.0057, tangentMode=RCTM_Break, time=0.8, value=1}
         <at 1 0 RCIM_Cubic>
        ]}}})
    ScaleA=true
    ScaleRGB=true
   )
  ]
  rendererClasses [&NiagaraSpriteRendererProperties]
  renderers [{rendererClass=&NiagaraSpriteRendererProperties}]
 }
 ParticleDecal {
  emitterUpdateScript [
   &EmitterState(
    Inactive Response=::NewEnumerator1
    Life Cycle Mode=::NewEnumerator1
    Loop Behavior=::NewEnumerator1
    Loop Duration Mode=::NewEnumerator1
    UseLoopDelay=false
   )
   &SpawnBurst_Instantaneous(Spawn Count=10, Use Spawn Probability=false, UseLoopCountLimit=false)
  ]
  particleSpawnScript [
   &PartitionParticles(Number Of Partitions=10)
   &InitializeParticle(
    Color=<rgba 1 0.0902 0.1002 1>
    Lifetime=3
    Sprite Rotation Mode=::NewEnumerator2
    Sprite Size Mode=::NewEnumerator3
    Uniform Sprite Size=&Multiply_Float(A=55, B=1)
    UsePositionOffset=false
    Write Material Random=true
   )
   &CameraOffset(Camera Offset Amount=50) {module=CameraOffset001}
   &DynamicMaterialParameters(
    Index 0 Param 1=1
    Index 0 Param 2=1
    Module.Index0Param0Write=true
    Module.Index0Param1Write=true
    Module.Index0Param2Write=true
    Module.Index0Param3Write=false
   )
   /SetVariables(Particles.EnergyDP:Vector4f=<2 1 8 1>) {isSetParameters=true}
  ]
  particleUpdateScript [
   &ParticleState(Kill Particles When Lifetime Has Elapsed=false, Let Infinitely Lived Particles Die When Emitter Deactivates=true)
   &ScaleColor(
    Scale Mode=::NewEnumerator1
    Scale RGBA=&MakeVector4_FromLinearColor(LinearColor=<rgba 1 0.5471 0.4973 1>)
    ScaleRGBA=true
   ) {module=ScaleColor001}
   &ScaleColor(
    Scale Alpha=&FloatFromCurve(
     FloatCurve=DataInterface() {propertyValues {Curve {keys [{tangentMode=RCTM_User, time=0, value=0}, {tangentMode=RCTM_User, time=1.8, value=1}]}}}
     CurveIndex=$Engine.System.Age
     Scale Curve=0.5
    )
    ScaleA=true
    ScaleRGB=true
   )
   /SetVariables(Particles.EnergyColor:LinearColor=&MultiplyLinearColor(
     Linear Color=$Particles.Color
     Scale Factor=&MakeVector4_FromFloat(Float=&Multiply_Float(A=&MakeCustomFloatFromBool(Bool=$User.MaskDBNO, True Float=0, False Float=1), B=30))
    )) {isSetParameters=true}
   &SkeletalMeshLocation(
    Bone / Socket Index=$Particles.PartitionParticles.Partition
    Bone Sampling Mode=::NewEnumerator4
    MaskSpawnGroup=false
    Sampled Position Offset=&MakeVector(
     X=&Multiply_Float(
      A=&SelectFloatFromArray(
       Direct Array Index=$Particles.PartitionParticles.Partition
       Float Selection Array=DataInterface() {propertyValues {FloatData [1,1,1,1,1,-1,-1,-1,-1,-1]}}
      )
      B=25
     )
     Y=0
     Z=0
    )
    Skeletal Mesh=$System.SK_Mesh
   )
   &ScaleColor(
    Scale Alpha=&MakeCustomFloatFromBool(Bool=$User.MaskDBNO, True Float=0, False Float=1)
    ScaleA=true
    ScaleRGB=true
   ) {module=ScaleColor002}
  ]
  rendererClasses [&NiagaraSpriteRendererProperties]
  renderers [
   {rendererClass=&NiagaraSpriteRendererProperties}
   {rendererClass=&NiagaraSpriteRendererProperties, rendererIndex=1}
   {rendererClass=&NiagaraSpriteRendererProperties, rendererIndex=2}
  ]
 }
 ParticleDecalWave {
  emitterUpdateScript [
   &EmitterState(Life Cycle Mode=::NewEnumerator1, Loop Duration=1, UseLoopDelay=false)
   &SpawnBurst_Instantaneous(Spawn Count=10, Use Spawn Probability=false, UseLoopCountLimit=false)
  ]
  particleSpawnScript [
   &InitializeParticle(
    Color=<rgba 1 0.1439 0.114 1>
    Lifetime=2
    Mass Mode=::NewEnumerator1
    Sprite Size Mode=::NewEnumerator3
    Uniform Sprite Size=100
    UsePositionOffset=false
    Write Material Random=true
   )
   &PartitionParticles(Number Of Partitions=10)
   &CameraOffset(Camera Offset Amount=30)
  ]
  particleUpdateScript [
   &ParticleState()
   &ScaleSpriteSize(
    Uniform Curve Scale=1.8
    Uniform Curve Sprite Scale=DataInterface() {propertyValues {Curve {keys [
        {arriveTangent=2.0583, interpMode=RCIM_Cubic, leaveTangent=2.0583, tangentMode=RCTM_User, time=0, value=0.45}
        {arriveTangent=0.55, interpMode=RCIM_Cubic, leaveTangent=0.55, time=0.2, value=0.7}
        {interpMode=RCIM_Cubic, tangentMode=RCTM_User, time=1, value=1}
       ]}}}
   ) {module=ScaleSpriteSize001}
   &ScaleColor(
    Scale Alpha=&FloatFromCurve(
     FloatCurve=DataInterface() {propertyValues {Curve {keys [{arriveTangent=-2.5, interpMode=RCIM_Cubic, time=0, value=0}, <at 0.1 1 RCIM_Cubic>, <at 1 0 RCIM_Cubic>]}}}
     Scale Curve=0.1
    )
    Scale RGB=&VectorFromFloat(Value=&FloatFromCurve(FloatCurve=DataInterface() {propertyValues {Curve {keys [
          {arriveTangent=9.0725, interpMode=RCIM_Cubic, leaveTangent=9.0725, tangentMode=RCTM_User, time=0, value=0}
          {arriveTangent=-2.5249, interpMode=RCIM_Cubic, leaveTangent=-2.5249, tangentMode=RCTM_User, time=0.4, value=0.5}
          {arriveTangent=-0.0405, interpMode=RCIM_Cubic, leaveTangent=-0.0405, tangentMode=RCTM_User, time=1, value=0}
         ]}}}))
    ScaleA=true
    ScaleRGB=true
   )
   &ScaleColor(
    Scale Alpha=&MakeCustomFloatFromBool(Bool=$User.MaskDBNO, True Float=0, False Float=1)
    ScaleA=true
    ScaleRGB=true
   ) {module=ScaleColor002}
   &SkeletalMeshLocation(
    Bone / Socket Index=$Particles.PartitionParticles.Partition
    Bone Sampling Mode=::NewEnumerator4
    MaskSpawnGroup=false
    Sampled Position Offset=&MakeVector(
     X=&Multiply_Float(
      A=&SelectFloatFromArray(
       Direct Array Index=$Particles.PartitionParticles.Partition
       Float Selection Array=DataInterface() {propertyValues {FloatData [1,1,1,1,1,-1,-1,-1,-1,-1]}}
      )
      B=25
     )
     Y=0
     Z=0
    )
    Skeletal Mesh=$System.SK_Mesh
   )
   &DynamicMaterialParameters(
    Index 0 Param 1=&FloatFromCurve(
     FloatCurve=DataInterface() {propertyValues {Curve {keys [{arriveTangent=1.4055, interpMode=RCIM_Cubic, leaveTangent=1.4055, tangentMode=RCTM_User, time=0, value=0}, <at 1 1 RCIM_Cubic>]}}}
     Scale Curve=8
    )
    Index 0 Param 2=&RandomRangeFloat(Minimum=0, Maximum=10)
    Index 0 Param 3=&FloatFromCurve(FloatCurve=DataInterface() {propertyValues {Curve {keys [{tangentMode=RCTM_User, time=0, value=0}, {tangentMode=RCTM_User, time=0.3, value=1}]}}})
    Index 0 Param 4=&FloatFromCurve(FloatCurve=DataInterface() {propertyValues {Curve {keys [{tangentMode=RCTM_User, time=0, value=0}, {tangentMode=RCTM_User, time=1, value=1}]}}})
    Module.Param0WriteEnabled=true
    Module.Param1WriteEnabled=true
    Module.Param2WriteEnabled=true
    Module.Param3WriteEnabled=true
   )
  ]
  rendererClasses [&NiagaraSpriteRendererProperties]
  renderers [{rendererClass=&NiagaraSpriteRendererProperties}]
 }
 StringsOnMaskHands {
  emitterSpawnScript [/SetVariables(Emitter.EmitterRandom:NiagaraFloat=&RandomRangeFloat(Minimum=0, Maximum=1)) {isSetParameters=true}]
  emitterUpdateScript [
   &EmitterState(Life Cycle Mode=::NewEnumerator1, Loop Behavior=::NewEnumerator1, Loop Duration Mode=::NewEnumerator1, UseLoopDelay=false)
   &SpawnBurst_Instantaneous(
    Spawn Count=10
    Spawn Probability=&MakeCustomFloatFromBool(Bool=$User.EnableStrings, True Float=1, False Float=0)
    Use Spawn Probability=true
    UseLoopCountLimit=false
   )
  ]
  particleSpawnScript [
   &PartitionParticles(Number Of Partitions=10)
   &InitializeParticle(Mesh Scale Mode=::NewEnumerator3, Position Mode=::NewEnumerator0, Write Material Random=true)
   &CameraOffset(Camera Offset Amount=20)
  ]
  particleUpdateScript [
   &ParticleState(Kill Particles When Lifetime Has Elapsed=false, Let Infinitely Lived Particles Die When Emitter Deactivates=true)
   /SetVariables(Particles.StringsOrigionPosition:NiagaraPosition=&AddVectorToPosition(
     Position=&AddVectorToPosition(
      Position=&LerpPosition(
       A=$User.StringTargetL
       B=$User.StringTargetR
       Alpha=&SelectFloatFromArray(
        Direct Array Index=$Particles.PartitionParticles.Partition
        Float Selection Array=DataInterface() {propertyValues {FloatData [1,1,1,1,1,0,0,0,0,0]}}
       )
      )
      Vector=&MakeVector(X=0, Y=0, Z=0)
     )
     Vector=&SelectVectorFromArray(
      Direct Array Index=$Particles.PartitionParticles.Partition
      Vector Selection Array=DataInterface() {propertyValues {FloatData [<-25 0 0>, <0 -25 0>, <15 0 -5>, <25 0 0>, <0 0 5>, <15 25 0>, <-25 0 0>, <0 -25 -5>, <15 0 0>, <25 0 5>]}}
     )
    )) {isSetParameters=true}
   &SkeletalMeshLocation(
    Bone / Socket Index=$Particles.PartitionParticles.Partition
    Bone Sampling Mode=::NewEnumerator4
    MaskSpawnGroup=false
    Orientation Sampling=::NewEnumerator1
    Sampled Position Offset=&MakeVector(
     X=&Multiply_Float(
      A=&SelectFloatFromArray(
       Direct Array Index=$Particles.PartitionParticles.Partition
       Float Selection Array=DataInterface() {propertyValues {FloatData [1,1,1,1,1,-1,-1,-1,-1,-1]}}
      )
      B=25
     )
     Y=0
     Z=0
    )
    Skeletal Mesh=$System.SK_Mesh
   )
   /SetVariables(Particles.StringLength:NiagaraFloat=&VectorLength(Vector=&SubtractVectorFromPosition(
      Position=$Particles.SkeletalMeshLocation.SampledPosition
      Vector=Linked() {linkedVariable {name=Particles.StringsOrigionPosition, type=NiagaraPosition}}
     ))) {isSetParameters=true}
   &ScaleMeshSize(Scale Factor=&MakeVector(X=&Divide_Float(A=$Particles.StringLength, B=100), Y=1, Z=1))
   /SetVariables(Particles.FadeOutWhenStretchTooLong:NiagaraFloat=&RemapRange(Input Value To Remap=$Particles.StringLength, Input Min=0, Input Max=5000, Output Min=0, Output Max=1)) {isSetParameters=true}
   /SetVariables(Particles.MeshOrientation=&OrientMeshToVector(
     Look At Direction=&Subtract_Position(A=$Particles.StringsOrigionPosition, B=$Particles.SkeletalMeshLocation.SampledPosition)
     Reference Axis Vector=<1 0 0>
    )) {isSetParameters=true}
   &DynamicMaterialParameters(
    Index 0 Param 1=&RandomRangeFloat(Minimum=0, Maximum=1)
    Index 0 Param 2=&MakeCustomFloatFromBool(
     Bool=$User.IsStringConnecting
     True Float=&FloatFromCurve(
      FloatCurve=DataInterface() {propertyValues {Curve {keys [
          {arriveTangent=-0.5, interpMode=RCIM_Cubic, time=0, value=1}
          {arriveTangent=0.176, time=0.3, value=0}
          <at 1.5 0.1>
          {arriveTangent=0.0224, interpMode=RCIM_Cubic, leaveTangent=0.0224, tangentMode=RCTM_User, time=2, value=0.2}
         ]}}}
      CurveIndex=$Particles.Age
     )
     False Float=0
    )
    Index 0 Param 3=&MakeCustomFloatFromBool(
     Bool=$User.IsStringConnecting
     True Float=&Lerp_Float(
      A=2
      B=8
      Alpha=&FloatFromCurve(
       FloatCurve=DataInterface() {propertyValues {Curve {keys [
           <at 0.0673 1 RCIM_Cubic>
           {arriveTangent=-0.0877, interpMode=RCIM_Cubic, leaveTangent=-0.0877, tangentMode=RCTM_User, time=0.18, value=0.145}
           <at 1 0 RCIM_Cubic>
          ]}}}
       CurveIndex=$Particles.Age
      )
     )
     False Float=&Lerp_Float(
      A=2
      B=5
      Alpha=&FloatFromCurve(
       FloatCurve=DataInterface() {propertyValues {Curve {keys [
           <at 0.0673 1 RCIM_Cubic>
           {arriveTangent=-0.0877, interpMode=RCIM_Cubic, leaveTangent=-0.0877, tangentMode=RCTM_User, time=0.18, value=0.145}
           <at 1 0 RCIM_Cubic>
          ]}}}
       CurveIndex=$Particles.Age
      )
     )
    )
    Index 0 Param 4=&MakeCustomFloatFromBool(
     Bool=$User.IsStringConnecting
     True Float=&Lerp_Float(
      A=1
      B=30
      Alpha=&FloatFromCurve(
       FloatCurve=DataInterface() {propertyValues {Curve {keys [
           <at 0 1 RCIM_Cubic>
           {arriveTangent=-3.9998, interpMode=RCIM_Cubic, leaveTangent=-3.9998, time=0.1, value=1}
           {arriveTangent=-5.1911, interpMode=RCIM_Cubic, leaveTangent=-5.1911, time=0.208, value=0.1681}
           {arriveTangent=-0.2754, interpMode=RCIM_Cubic, leaveTangent=-0.0154, tangentMode=RCTM_Break, time=0.301, value=-0.0434}
           {arriveTangent=0.1805, interpMode=RCIM_Cubic, leaveTangent=0.1805, time=0.4658, value=0.0168}
           {interpMode=RCIM_Cubic, leaveTangent=0.0603, time=0.7, value=0.0286}
          ]}}}
       CurveIndex=$Particles.Age
      )
     )
     False Float=&Lerp_Float(
      A=1
      B=25
      Alpha=&FloatFromCurve(
       FloatCurve=DataInterface() {propertyValues {Curve {keys [
           {arriveTangent=3.4732, interpMode=RCIM_Cubic, leaveTangent=3.4732, tangentMode=RCTM_User, time=0.0015, value=0.0401}
           {arriveTangent=2.2094, interpMode=RCIM_Cubic, leaveTangent=2.2094, time=0.1486, value=0.4292}
           {arriveTangent=0.9016, interpMode=RCIM_Cubic, leaveTangent=0.9016, time=0.3666, value=0.8467}
           {interpMode=RCIM_Cubic, leaveTangent=0.0603, time=0.697, value=0.9237}
          ]}}}
       CurveIndex=$Particles.Age
      )
     )
    )
    Index 1 Param 1=&MakeCustomFloatFromBool(
     Bool=$User.IsStringConnecting
     True Float=0
     False Float=&FloatFromCurve(
      FloatCurve=DataInterface() {propertyValues {Curve {keys [
          {arriveTangent=1.2663, interpMode=RCIM_Cubic, leaveTangent=1.2663, tangentMode=RCTM_User, time=0, value=0}
          {arriveTangent=0.7394, interpMode=RCIM_Cubic, leaveTangent=0.7394, time=0.3246, value=0.3858}
          {arriveTangent=0.9094, interpMode=RCIM_Cubic, leaveTangent=0.9094, time=0.7226, value=0.5343}
          {arriveTangent=0.1468, interpMode=RCIM_Cubic, leaveTangent=0.1468, tangentMode=RCTM_User, time=1, value=1}
         ]}}}
      Scale Curve=3.5
     )
    )
    Module.Index0Param0Write=true
    Module.Index0Param1Write=true
    Module.Index0Param2Write=true
    Module.Index0Param3Write=true
    Module.Index1Param0Write=true
    Module.Index1Param1Write=true
    Module.Index1Param2Write=true
    Module.Index1Param3Write=true
    Write Parameter Index 1=true
   ) {module=DynamicMaterialParameters002}
   &ScaleColor(
    Scale Alpha=&FloatFromCurve(
     FloatCurve=DataInterface() {propertyValues {Curve {keys [{tangentMode=RCTM_User, time=0, value=1}, <at 0.8 1>, {tangentMode=RCTM_User, time=1, value=0}]}}}
     CurveIndex=$Particles.FadeOutWhenStretchTooLong
    )
    Scale RGB=&VectorFromFloat(Value=&Multiply_Float(
      A=&FloatFromCurve(
       FloatCurve=DataInterface() {propertyValues {Curve {keys [
           {interpMode=RCIM_Cubic, leaveTangent=4.915, tangentMode=RCTM_Break, time=0.0031, value=0}
           {arriveTangent=0.2458, interpMode=RCIM_Cubic, leaveTangent=-2.1732, tangentMode=RCTM_Break, time=0.28, value=1}
           {arriveTangent=-1.2375, leaveTangent=-1.2375, time=0.38, value=0.01}
           {interpMode=RCIM_Cubic, tangentMode=RCTM_Break, time=0.65, value=0.001}
          ]}}}
       CurveIndex=$Particles.Age
      )
      B=200
     ))
    ScaleA=true
    ScaleRGB=true
   )
   &ScaleColor(
    Linear Color Curve=DataInterface() {propertyValues {
      AlphaCurve {keys [<at 0 1>]}
      BlueCurve {keys [<at 0.4298 0.0589>,<at 0.5741 1>]}
      GreenCurve {keys [<at 0.4298 0.0556>,<at 0.5741 1>]}
      RedCurve {keys [<at 0.4298 1>,<at 0.5741 1>]}
     }}
    Scale Mode=::NewEnumerator2
   ) {module=ScaleColor001}
  ]
  rendererClasses [&NiagaraMeshRendererProperties]
  renderers [{rendererClass=&NiagaraMeshRendererProperties, rendererIndex=0}]
 }
 WashDecals {
  emitterUpdateScript [
   &EmitterState(Life Cycle Mode=::NewEnumerator1, Loop Behavior=::NewEnumerator1, Loop Duration=1, UseLoopDelay=false)
   &SpawnBurst_Instantaneous(Spawn Count=10, Spawn Time=0.05, Use Spawn Probability=false)
  ]
  particleSpawnScript [
   &PartitionParticles(Number Of Partitions=10)
   &InitializeParticle(Sprite Size Mode=::NewEnumerator3, Uniform Sprite Size=150, UsePositionOffset=false, Write Material Random=true)
   &CameraOffset(Camera Offset Amount=10)
  ]
  particleUpdateScript [
   &ParticleState()
   &ScaleColor(
    Scale Alpha=&FloatFromCurve(FloatCurve=DataInterface() {propertyValues {Curve {keys [
         {arriveTangent=0.3356, interpMode=RCIM_Cubic, leaveTangent=0.3356, tangentMode=RCTM_User, time=0.1, value=1.1}
         {tangentMode=RCTM_User, time=1, value=0}
        ]}}})
    Scale RGB=&VectorFromFloat(Value=&FloatFromCurve(
      FloatCurve=DataInterface() {propertyValues {Curve {keys [
          {arriveTangent=-2.1358, interpMode=RCIM_Cubic, leaveTangent=-2.1358, tangentMode=RCTM_User, time=0, value=0.9847}
          {arriveTangent=-0.0405, interpMode=RCIM_Cubic, leaveTangent=-0.0405, tangentMode=RCTM_User, time=1, value=0}
         ]}}}
      Scale Curve=5
     ))
    ScaleA=true
    ScaleRGB=true
   ) {module=ScaleColor002}
   &DynamicMaterialParameters(
    Index 0 Param 1=&FloatFromCurve(
     FloatCurve=DataInterface() {propertyValues {Curve {keys [
         {interpMode=RCIM_Cubic, leaveTangentWeight=0.0435, tangentMode=RCTM_User, time=0, value=-0.2}
         {arriveTangent=18.2867, interpMode=RCIM_Cubic, leaveTangent=18.2867, tangentMode=RCTM_User, time=0.07, value=0.275}
         {arriveTangent=5.0806, interpMode=RCIM_Cubic, leaveTangent=5.0806, time=0.15, value=1}
         {arriveTangent=5.6764, interpMode=RCIM_Cubic, leaveTangent=5.6764, tangentMode=RCTM_User, time=1, value=5}
        ]}}}
     Scale Curve=0.5
    )
    Index 0 Param 2=1
    Index 0 Param 3=&FloatFromCurve(FloatCurve=DataInterface() {propertyValues {Curve {keys [{arriveTangent=2.2883, interpMode=RCIM_Cubic, leaveTangent=2.2883, tangentMode=RCTM_User, time=0, value=-0.1}, <at 1 1.8 RCIM_Cubic>]}}})
    Module.Param0WriteEnabled=true
    Module.Param1WriteEnabled=true
    Module.Param2WriteEnabled=true
    Module.Param3WriteEnabled=false
   )
   &SkeletalMeshLocation(
    Bone / Socket Index=$Particles.PartitionParticles.Partition
    Bone Sampling Mode=::NewEnumerator4
    MaskSpawnGroup=false
    Sampled Position Offset=&MakeVector(
     X=&Multiply_Float(
      A=&SelectFloatFromArray(
       Direct Array Index=$Particles.PartitionParticles.Partition
       Float Selection Array=DataInterface() {propertyValues {FloatData [1,1,1,1,1,-1,-1,-1,-1,-1]}}
      )
      B=40
     )
     Y=0
     Z=0
    )
    Skeletal Mesh=$System.SK_Mesh
   )
   &ScaleColor(
    Linear Color Curve=DataInterface() {propertyValues {
      AlphaCurve {keys [<at 0 1>,<at 1 0>]}
      BlueCurve {keys [<at 0.334 0.0481>,<at 0.7364 1>]}
      GreenCurve {keys [<at 0.334 0.0381>,<at 0.7364 1>]}
      RedCurve {keys [<at 0.334 1>,<at 0.7364 1>]}
     }}
    Scale Mode=::NewEnumerator2
   )
  ]
  rendererClasses [&NiagaraSpriteRendererProperties]
  renderers [{rendererClass=&NiagaraSpriteRendererProperties}]
 }
}
summary {
 systemName=NS_PuppeteerBoss_StringsOnMask
 emitters [
  {emitterName=StringsOnMaskHands, rendererClasses [&NiagaraMeshRendererProperties]}
  {emitterName=WashDecals, rendererClasses [&NiagaraSpriteRendererProperties]}
  {emitterName=ParticleDecal, rendererClasses [&NiagaraSpriteRendererProperties]}
  {emitterName=ParticleDecalWave, rendererClasses [&NiagaraSpriteRendererProperties]}
  {emitterName=EyeSparkles, simTarget=GPUComputeSim, rendererClasses [&NiagaraSpriteRendererProperties]}
  {emitterName=EyeFlash, rendererClasses [&NiagaraSpriteRendererProperties, &NiagaraLightRendererProperties]}
  {emitterName=IdleSmoke, rendererClasses [&NiagaraSpriteRendererProperties]}
 ]
 userVariables [
  User.EnableStrings {type=NiagaraBool, defaultValue=true}
  User.EnableWeakpoints {type=NiagaraBool, defaultValue=true}
  User.HideLeftEyeFX {type=NiagaraBool}
  User.HideRightEyeFX {type=NiagaraBool}
  User.IsLeftEye {type=NiagaraBool}
  User.IsStringConnecting {type=NiagaraBool, defaultValue=true}
  User.MaskDBNO {type=NiagaraBool}
  User.PlayHitFX {type=NiagaraBool}
  User.StringTargetL {type=NiagaraPosition, description="", defaultValue {}}
  User.StringTargetOffset {type=NiagaraFloat, defaultValue=0, description=""}
  User.StringTargetR {type=NiagaraPosition, description="", defaultValue {}}
 ]
}
