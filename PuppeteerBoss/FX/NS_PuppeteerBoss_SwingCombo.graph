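# NS_PuppeteerBoss_SwingCombo is a Niagara effect for the Puppeteer Boss's swing combo, drawing a short-lived triple claw slash.
# One Claw emitter bursts three flattened mesh particles, using partition index to pick vertically stacked offsets from an array.
# Two color-scale passes combine a red-tinted gradient with an intense, fast-decaying brightness and alpha fade.
# Randomized dynamic material parameters vary the slashes; the declared swing-section and impact-direction inputs are not visibly read by the emitter.
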NiagaraSystem /PuppeteerBoss/FX/NS_PuppeteerBoss_SwingCombo
dependencies {
 usedDataInterfaces [&NiagaraDataInterfaceArrayFloat3, &NiagaraDataInterfaceColorCurve, &NiagaraDataInterfaceCurve]
 usedDynamicInputs [&MakeVector, &FN_LOD_Distance_Remap]
 usedModules [
  &SystemState
  &self:SystemSpawnScript.SystemScriptSource.SystemScriptGraph.NiagaraNodeAssignment_0.SetVariables_#1
 ]
}
emitters {Claw {
  emitterUpdateScript [
   &EmitterState(Life Cycle Mode=::NewEnumerator1, Loop Behavior=::NewEnumerator1, Loop Duration=0.5, UseLoopDelay=false)
   &SpawnBurst_Instantaneous(Spawn Count=3, Use Spawn Probability=false, UseLoopCountLimit=false)
  ]
  particleSpawnScript [
   &PartitionParticles(Number Of Partitions=3)
   &InitializeParticle(Lifetime=0.6, Mesh Scale=<0.7 0.7 0.08>, Mesh Scale Mode=::NewEnumerator4, UsePositionOffset=false, Write Material Random=true)
   &InitialMeshOrientation(Mesh Orientation Mode=::NewEnumerator5, Rotation=<0 0 -0.25>, Use Rotation Vector=true)
   &SystemLocation(
    MaskSpawnGroup=false
    Offset=&SelectVectorFromArray(
     Direct Array Index=$Particles.PartitionParticles.Partition
     Vector Selection Array=DataInterface() {propertyValues {FloatData [<0 0 5>, <0 0 0>, <0 0 -5>]}}
    )
   )
  ]
  particleUpdateScript [
   &ParticleState()
   &ScaleColor(
    Linear Color Curve=DataInterface() {propertyValues {
      AlphaCurve {keys [<at 0 1>]}
      BlueCurve {keys [<at 0.0294 0.1145>,<at 0.1808 0.0699>,<at 0.6256 1>]}
      GreenCurve {keys [<at 0.0294 0>,<at 0.1808 0.0157>,<at 0.6256 1>]}
      RedCurve {keys [<at 0.0294 5>,<at 0.1808 1>,<at 0.6256 1>]}
     }}
    Scale Mode=::NewEnumerator2
   ) {module=ScaleColor001}
   &ScaleColor(
    Scale Alpha=&FloatFromCurve(FloatCurve=DataInterface() {propertyValues {Curve {keys [
         {arriveTangent=-1.4485, interpMode=RCIM_Cubic, leaveTangent=-1.4485, tangentMode=RCTM_User, time=0, value=1}
         {arriveTangent=-0.406, interpMode=RCIM_Cubic, leaveTangent=-0.406, tangentMode=RCTM_User, time=0.3879, value=0.1559}
         {interpMode=RCIM_Cubic, leaveTangent=-8.0089, time=1, value=0}
        ]}}})
    Scale RGB=&VectorFromFloat(Value=&FloatFromCurve(FloatCurve=DataInterface() {propertyValues {Curve {keys [<at 0.05 1>,<at 0.15 0.15>,<at 1 0>]}}}, Scale Curve=20))
    ScaleA=true
    ScaleRGB=true
   )
   &DynamicMaterialParameters(
    Index 0 Param 1=&RandomRangeFloat(Minimum=0, Maximum=1)
    Module.Index0Param0Write=true
    Module.Index0Param1Write=true
    Module.Index0Param2Write=true
    Module.Index0Param3Write=true
   )
  ]
  rendererClasses [&NiagaraMeshRendererProperties]
  renderers [{rendererClass=&NiagaraMeshRendererProperties, rendererIndex=0}]
 }}
summary {
 systemName=NS_PuppeteerBoss_SwingCombo
 emitters [{emitterName=Claw, rendererClasses [&NiagaraMeshRendererProperties]}]
 userVariables [
  User.bIsFrontEnd {type=NiagaraBool}
  User.bIsFrontEndPreview {type=NiagaraBool}
  User.ImpactFXDirection {type=Vector3f, description="", defaultValue {struct=Vector3f, value=<0 0 0>}}
  User.SwingSectionIndex {type=NiagaraInt32, defaultValue=-1}
 ]
}
