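# Base material for the Puppeteer boss, adding glowing eye weak points and a timed hide effect to the shared character parent functions.
# Pre-skinned sphere masks locate each eye, where custom primitive data toggles a Fresnel-based two-color glow.
# A timestamp and flag in custom primitive data drive the hide animation, alongside temporal-dithered opacity.
# Character PreFX supplies the base attributes, while PostFX feeds a demon-boss derez function as the last stage.
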Material /PuppeteerBoss/Materials/M_AI_PuppeteerBoss_Base
graph {
 properties {
  bUseMaterialAttributes=true
  bUsedWithClothing=true
  bUsedWithMeshDeformer=true
  bUsedWithMorphTargets=true
  bUsedWithNanite=true
  bUsedWithNiagaraMeshParticles=true
  bUsedWithSkeletalMesh=true
  pixelDepthOffsetMode=PDOM_Legacy
  shadingModel=MSM_FromMaterialExpression
  usedShadingModels=Default Lit | Subsurface | Clear Coat | Subsurface Profile | Cloth
 }
 nodes [
  BlendMaterialAttributes(
   B=SetMaterialAttributes!(MaterialAttributes, Opacity Mask=Constant()) {attributeSetTypes [#1]}
   Alpha=NamedRerouteUsage()
  )
  n1 = LocalPosition() {localOrigin=InstancePreSkinning}
  ComponentMask(None=@n1.XYZ) {g=true}
  n2 = MaterialFunctionCall() {materialFunction=&MF_CharacterParent_2023_PreFX}
  MaterialFunctionCall!(Material=MaterialFunctionCall(
    Material=NamedRerouteUsage()
    M=@n2.M
    Pre Skinned Local Position=NamedRerouteDeclaration(None=@n2.Pre skinned Local Position) {name=PreskinnedPosition, nodeColor=<rgba 1 0 0.2812 1>}
    Distance Blend=@n2.M
    Fresnel=@n2.M
   ) {materialFunction=&MF_CharacterParent_2023_PostFX}.Result) {materialFunction=&/MF_DemonBoss_Derez}
  NamedRerouteDeclaration(None=MaterialFunctionCall(Alpha Threshold=NamedRerouteUsage()) {materialFunction=&DitherTemporalAA}.Result) {name=MA_Hidden, nodeColor=<rgba 1 0 0.2578 1>}
  NamedRerouteDeclaration(None=@n2.Result) {name=MA_In, nodeColor=<rgba 0.7109 0 1 1>}
  n3 = NamedRerouteUsage()
  n4 = Constant()
  n5 = GetMaterialAttributes(None=@n3) {attributeGetTypes [#2]}
  n6 = VectorParameter() {group=WeakPoints, parameterName=InnerGlowColor, defaultValue=<rgba 0.9911 0.8141 0.0052 1>}
  n7 = ScalarParameter() {defaultValue=1, group=WeakPoints, parameterName=Glow Intensity}
  n8 = VertexNormalWS()
  n9 = CameraVectorWS()
  n10 = OneMinus(None=Abs(None=DotProduct(A=@n8, B=@n9)))
  n11 = Saturate(None=ScalarParameter() {defaultValue=0.75, group=WeakPoints, parameterName=Glow Size})
  n12 = OneMinus(None=Saturate(None=Divide(A=@n10, B=@n11)))
  n13 = TextureSampleParameter2D() {group=WeakPoints, parameterName=EmissiveMask, sortPriority=0, texture=/Game/Global/Textures/Default/Blanks/T_Blank_LegalWhite_sRGBColor}
  n14 = Multiply(
   A=LinearInterpolate(
    A=VectorParameter() {group=WeakPoints, parameterName=OuterGlowColor, defaultValue=<rgba 0.8308 0.0012 0.0044 1>}.RGB
    B=Multiply(A=@n6.RGB, B=@n7)
    Alpha=@n12
   )
   B=@n13.R
  )
  n15 = LinearInterpolate(A=@n5.Base Color, B=Saturate(None=@n14), Alpha=@n13.R)
  n16 = Multiply(A=NamedRerouteUsage(), B=ScalarParameter() {bUseCustomPrimitiveData=true, parameterName=GlowRightEye_CPD, primitiveDataIndex=3})
  n17 = Multiply(A=NamedRerouteUsage(), B=ScalarParameter() {bUseCustomPrimitiveData=true, parameterName=GlowLeftEye_CPD, primitiveDataIndex=4})
  NamedRerouteDeclaration(None=BlendMaterialAttributes(
    A=SetMaterialAttributes!(MaterialAttributes=@n3, Emissive Color=@n4) {attributeSetTypes [#3]}
    B=SetMaterialAttributes!(MaterialAttributes=@n5.MaterialAttributes, Base Color=@n15, Emissive Color=@n14) {attributeSetTypes [#2, #3]}
    Alpha=Max(A=@n16, B=@n17)
   )) {name=MA_Out, nodeColor=<rgba 0 1 0.3906 1>}
  n18 = ScalarParameter() {bUseCustomPrimitiveData=true, parameterName=HideTimeStamp_CPD, primitiveDataIndex=1}
  n19 = Abs(None=Subtract(A=Time(), B=@n18))
  n20 = ScalarParameter() {defaultValue=1, group=Puppeteer, parameterName=HideAnimSpeed}
  n21 = Multiply(A=Saturate(None=Divide(A=@n19, B=@n20)), B=Saturate(None=@n18))
  n22 = LinearInterpolate(A=OneMinus(None=@n21), B=@n21, Alpha=ScalarParameter() {bUseCustomPrimitiveData=true, parameterName=HideFlag_CPD, primitiveDataIndex=2})
  NamedRerouteDeclaration(None=SquareRoot(None=NamedRerouteDeclaration(None=@n22) {name=Time, nodeColor=<rgba 1 0.2109 0 1>})) {name=SquaredTime, nodeColor=<rgba 0 0.1094 1 1>}
  n23 = SphereMask(
   A=NamedRerouteUsage()
   B=VectorParameter() {group=WeakPoints, parameterName=REyeLocation, defaultValue=<rgba -3.9221 8.5937 29.0654 1>}.RGB
   Radius=ScalarParameter() {defaultValue=2.36, group=WeakPoints, parameterName=EyeMaskRadius}
   Hardness=ScalarParameter() {defaultValue=1, group=WeakPoints, parameterName=EyeMaskHardness}
  ) {attenuationRadius=3}
  n24 = SphereMask(
   A=NamedRerouteUsage()
   B=VectorParameter() {group=WeakPoints, parameterName=LEyeLocation, defaultValue=<rgba 3.9221 8.5937 29.0654 1>}.RGB
   Radius=ScalarParameter() {defaultValue=2.36, group=WeakPoints, parameterName=EyeMaskRadius}
   Hardness=ScalarParameter() {defaultValue=1, group=WeakPoints, parameterName=EyeMaskHardness}
  ) {attenuationRadius=3}
  NamedRerouteDeclaration(None=Add(
    A=NamedRerouteDeclaration(None=@n23) {name=Mask_REye, nodeColor=<rgba 0 1 0.1797 1>}
    B=NamedRerouteDeclaration(None=@n24) {name=Mask_LEye, nodeColor=<rgba 1 0 0.1406 1>}
   )) {name=Mask_BothEyes, nodeColor=<rgba 0 0.4609 1 1>}
  NamedRerouteDeclaration(None=Power(Base=@n12, Exp=ScalarParameter() {defaultValue=22, group=WeakPoints, parameterName=Center Glow Exponent})) {name=innerGlow, nodeColor=<rgba 0.5469 1 0 1>}
  n25 = Divide(A=OneMinus(None=@n1.Z)) {constB=512}
  Subtract(A=ShaderStageSwitch(PixelShader=VertexInterpolator(VS=@n25).PS, VertexShader=@n25) {desc=glider gradient}) {constB=0.5}
 ]
}
