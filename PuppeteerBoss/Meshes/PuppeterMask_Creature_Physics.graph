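# Physics asset for the Puppeteer boss's mask creature mesh, giving capsule collision bodies to its mask and two hands.
# The mask body is a compound of six capsules, with separate bodies for the jaw and each eye.
# Each hand has a palm body plus one per metacarpal bone, mirrored left to right, with right-side finger and eye shapes keeping left-side names.
# It defines no constraints, so no joints link the bodies.
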PhysicsAsset /PuppeteerBoss/Meshes/PuppeterMask_Creature_Physics
bodies {
 hand_l [{shapeType=Capsule, length=2.8539, radius=2.8511, center=<-3.1442 0.2544 0.3193>, rotation=<rot 0.0007 -0.0002 -100>}]
 hand_r [{shapeType=Capsule, length=2.8539, radius=2.8511, center=<3.1442 -0.2544 -0.3193>, rotation=<rot 0.0007 -0.0002 80>}]
 index_metacarpal_l [{
   shapeType=Capsule
   length=7.1473
   radius=1.7405
   shapeName=index_metacarpal_l_capsule
   center=<-6.7257 0 0.9476>
   rotation=<rot -75.0002 179.9998 -179.9999>
  }]
 index_metacarpal_r [{
   shapeType=Capsule
   length=7.1473
   radius=1.7405
   shapeName=index_metacarpal_l_capsule
   center=<6.7257 0 -0.9476>
   rotation=<rot -75.0002 179.9998 0.0001>
  }]
 mask [
  {shapeType=Capsule, length=6.2616, radius=2.284, center=<4.3357 -0.0001 9.3054>, rotation=<rot 85.9246 -179.9762 -89.9761>}
  {shapeType=Capsule, length=11.8013, radius=3.7497, shapeName=mask_capsule, center=<7.9092 0 6.2783>, rotation=<rot 0.0002 0 90>}
  {shapeType=Capsule, length=0, radius=3.3478, shapeName=mask_capsule, center=<7.852 0 0>, rotation=<rot 0.0003 0 90>}
  {
   shapeType=Capsule
   length=15.6697
   radius=1.9945
   shapeName=mask_capsule
   center=<3.6799 -8.2119 -0.3618>
   rotation=<rot 0.189 0.8475 -14.9227>
  }
  {shapeType=Capsule, length=15, radius=2, shapeName=mask_capsule, center=<3.1184 8.0753 0.4733>, rotation=<rot 0.0007 0 15>}
  {shapeType=Capsule, length=9.5, radius=3, shapeName=mask_capsule, center=<7.5535 0 -4.6344>, rotation=<rot 0.0002 0 90>}
 ]
 mask_eye_l [{shapeType=Capsule, length=0.0973, radius=2.9904, shapeName=mask_eye_l_capsule, center=<0.8728 0 0>}]
 mask_eye_r [{shapeType=Capsule, length=0.0973, radius=2.9904, shapeName=mask_eye_l_capsule, center=<0.8728 0 0>, rotation=<rot 0 0 -180>}]
 mask_jaw [{shapeType=Capsule, length=6.9108, radius=3.3737, shapeName=mask_jaw_capsule, center=<7.8348 0 -7.813>, rotation=<rot 0.0003 0 -90>}]
 middle_metacarpal_l [{
   shapeType=Capsule
   length=7.0493
   radius=1.6003
   shapeName=middle_metacarpal_l_capsule
   center=<-7.2374 0 0.7512>
   rotation=<rot -75.0002 179.9996 -180>
  }]
 middle_metacarpal_r [{
   shapeType=Capsule
   length=7.0493
   radius=1.6003
   shapeName=middle_metacarpal_l_capsule
   center=<7.2374 0 -0.7512>
   rotation=<rot -75.0002 179.9996 0>
  }]
 pinky_metacarpal_l [{
   shapeType=Capsule
   length=6.4768
   radius=1.6431
   shapeName=pinky_metacarpal_l_capsule
   center=<-5.5709 -0.4877 0.5021>
   rotation=<rot -75.0002 -175.0006 179.9999>
  }]
 pinky_metacarpal_r [{
   shapeType=Capsule
   length=6.4768
   radius=1.6431
   shapeName=pinky_metacarpal_l_capsule
   center=<5.5709 0.4877 -0.5021>
   rotation=<rot -75.0002 -175.0006 -0.0001>
  }]
 ring_metacarpal_l [{
   shapeType=Capsule
   length=7.9706
   radius=1.6288
   shapeName=ring_metacarpal_l_capsule
   center=<-7.2046 0.0001 0.5614>
   rotation=<rot -80.0002 179.9995 180>
  }]
 ring_metacarpal_r [{
   shapeType=Capsule
   length=7.9706
   radius=1.6288
   shapeName=ring_metacarpal_l_capsule
   center=<7.2046 -0.0001 -0.5614>
   rotation=<rot -80.0002 179.9995 0>
  }]
 thumb_metacarpal_l [{
   shapeType=Capsule
   length=4.5515
   radius=1.84
   shapeName=thumb_metacarpal_l_capsule
   center=<-3.7348 -0.344 0.0338>
   rotation=<rot -82.0704 -62.3907 52.6027>
  }]
 thumb_metacarpal_r [{
   shapeType=Capsule
   length=4.5515
   radius=1.84
   shapeName=thumb_metacarpal_l_capsule
   center=<3.7348 0.344 -0.0338>
   rotation=<rot -82.0704 -62.3907 -127.3973>
  }]
}
constraints []
