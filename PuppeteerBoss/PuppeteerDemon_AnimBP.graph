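# Puppeteer boss demon AnimBP on the Fortnite player skeleton; its only real logic is the ItemFinalPoseOverride layer.
# While bPBA is true, that layer sets world-space root yaw from DemonHangingRot and, during bIsStartPhase1, blends in the Teleport state machine.
# The machine eases from passthrough into a half-weight attached additive, then alternates two identical teleport states on bIsTeleport to replay the slowed additive.
# Otherwise the input pose passes through.
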AnimBlueprint /PuppeteerBoss/PuppeteerDemon_AnimBP
target_skeleton=&Fortnite_M_Avg_Player_Skeleton
graphs [
 AnimGraph {nodes [Output Pose()]}
 EventGraph {links [], nodes []}
 ItemAimOffset {nodes [
   Input Pose!() -> (Pose, AimOffsetAlpha:double, Yaw:double, Pitch:double) {name=InPose_AimOffset}
   Input Pose!() -> (Pose) {name=InPose_PreAimOffsets}
   Output Pose() {layerGroup=Item Override}
  ]}
 ItemFallAdditive {nodes [
   Input Pose() {name=InPoseItemFallAdditive}
   Output Pose() {layerGroup=Item Override}
  ]}
 ItemFallLandAdditive {nodes [
   Input Pose() {name=InPoseItemFallLandAdditive}
   Output Pose() {layerGroup=Item Override}
  ]}
 ItemFallLevel2Override {nodes [
   Input Pose() {name=InPose_FallLevel2}
   Output Pose() {layerGroup=Item Override}
  ]}
 ItemFinalPoseOverride {nodes [
   Apply Additive() {alphaBoolBlend {blendOption=Linear}}
   Blend Poses by bool(BlendPose_1=Use cached pose 'Final Input Pose'(), BlendTime_0=1, BlendTime_1=0.5, bActiveValue=Get bPBA() -> (bPBA:bool)) {blendType=Linear}
   AnimGraphNode_StateMachine_1 = Teleport SM() {is=title AnimGraphNode_StateMachine}
   n1 = Use cached pose 'Final Input Pose'()
   n2 = Get bIsStartPhase1() -> (bIsStartPhase1:bool)
   n3 = "Transform (Modify) Bone - Bone: Root"!(
    ComponentPose=Local To Component(LocalPose=Blend Poses by bool(BlendPose_0=@AnimGraphNode_StateMachine_1, BlendPose_1=@n1, BlendTime_0=0.5, BlendTime_1=0.5, bActiveValue=@n2) {blendType=Linear})
    Alpha=1
    Translation="0,0,0"
    Rotation_Roll:float=0
    Rotation_Pitch:float=0
    Rotation_Yaw:float=Get DemonHangingRot() -> (DemonHangingRot:float)
    Scale="1,1,1"
   ) -> (Pose) {rotationMode=BMM_Replace, rotationSpace=BCS_WorldSpace, alphaBoolBlend {blendOption=Linear}, boneToModify {boneName=Root}}
   Output Pose(Result=Blend Poses by bool(
     BlendPose_0=Component To Local(ComponentPose=@n3)
     BlendPose_1=Use cached pose 'Final Input Pose'()
     BlendTime_0=0.5
     BlendTime_1=0.5
     bActiveValue=Get bPBA()
    ) {blendType=Linear}) {layerGroup=Item Override}
   Save cached pose 'Final Input Pose'(Pose=Input Pose() {name=InFinalPose}) {cacheName=Final Input Pose}
   Get bIsMaskInactive() -> (bIsMaskInactive:bool)
  ]}
 Teleport SM {
  entry=Passthrough
  path=ItemFinalPoseOverride.AnimGraphNode_StateMachine_1.Teleport SM
  ids=id
  nodes [
   AnimStateNode_0 = Passthrough() {is=title AnimStateNode}
   AnimStateNode_1 = Teleport_1() {is=title AnimStateNode, bAlwaysResetOnEntry=true}
   AnimStateNode_2 = Teleport() {is=title AnimStateNode}
   AnimStateNode_3 = Attached() {is=title AnimStateNode}
   AnimStateTransitionNode_0 = AnimStateTransitionNode() {crossfadeDuration=0}
   AnimStateTransitionNode_3 = AnimStateTransitionNode() {crossfadeDuration=0}
   AnimStateTransitionNode_4 = AnimStateTransitionNode() {crossfadeDuration=0}
   AnimStateTransitionNode_5 = AnimStateTransitionNode() {crossfadeDuration=2}
  ]
  links [
   {
    dst=Teleport
    rule=ItemFinalPoseOverride.AnimGraphNode_StateMachine_1.Teleport SM.AnimStateTransitionNode_0.Transition
    src=Attached
    via=AnimStateTransitionNode_0
   }
   {
    dst=Attached
    rule=ItemFinalPoseOverride.AnimGraphNode_StateMachine_1.Teleport SM.AnimStateTransitionNode_5.Transition
    src=Passthrough
    via=AnimStateTransitionNode_5
   }
   {
    dst=Teleport_1
    rule=ItemFinalPoseOverride.AnimGraphNode_StateMachine_1.Teleport SM.AnimStateTransitionNode_3.Transition
    src=Teleport
    via=AnimStateTransitionNode_3
   }
   {
    dst=Teleport
    rule=ItemFinalPoseOverride.AnimGraphNode_StateMachine_1.Teleport SM.AnimStateTransitionNode_4.Transition
    src=Teleport_1
    via=AnimStateTransitionNode_4
   }
  ]
 }
 Passthrough {
  path=ItemFinalPoseOverride.AnimGraphNode_StateMachine_1.Teleport SM.AnimStateNode_0.Passthrough
  nodes [Output Animation Pose(Result=Use cached pose 'Final Input Pose'())]
 }
 Teleport_1 {
  path=ItemFinalPoseOverride.AnimGraphNode_StateMachine_1.Teleport SM.AnimStateNode_1.Teleport_1
  nodes [Output Animation Pose(Result=Apply Additive(
     Base=Use cached pose 'Final Input Pose'()
     Additive="Sequence Player (additive) 'Puppeteer_Teleport_Additive'"() {bLoopAnimation=false, playRate=0.3, sequence=/FirePetalBosses/Animation/Puppeteer_Teleport_Additive}
    ) {alphaBoolBlend {blendOption=Linear}})]
 }
 Teleport {
  path=ItemFinalPoseOverride.AnimGraphNode_StateMachine_1.Teleport SM.AnimStateNode_2.Teleport
  nodes [Output Animation Pose(Result=Apply Additive(
     Base=Use cached pose 'Final Input Pose'()
     Additive="Sequence Player (additive) 'Puppeteer_Teleport_Additive'"() {bLoopAnimation=false, playRate=0.3, sequence=&Puppeteer_Teleport_Additive}
    ) {alphaBoolBlend {blendOption=Linear}})]
 }
 Attached {
  path=ItemFinalPoseOverride.AnimGraphNode_StateMachine_1.Teleport SM.AnimStateNode_3.Attached
  nodes [Output Animation Pose(Result=Apply Additive(
     Base=Use cached pose 'Final Input Pose'()
     Additive="Sequence Player (additive) 'Puppeteer_Teleport_Additive'"() {bLoopAnimation=false, playRate=0.3, sequence=&Puppeteer_Teleport_Additive}
     Alpha=0.5
    ) {alphaBoolBlend {blendOption=Linear}})]
 }
 Transition {
  path=ItemFinalPoseOverride.AnimGraphNode_StateMachine_1.Teleport SM.AnimStateTransitionNode_0.Transition
  nodes [Result(bCanEnterTransition=Get bIsTeleport() -> (bIsTeleport:bool))]
 }
 Transition {
  path=ItemFinalPoseOverride.AnimGraphNode_StateMachine_1.Teleport SM.AnimStateTransitionNode_3.Transition
  nodes [Result(bCanEnterTransition=Get bIsTeleport())]
 }
 Transition {
  path=ItemFinalPoseOverride.AnimGraphNode_StateMachine_1.Teleport SM.AnimStateTransitionNode_4.Transition
  nodes [Result(bCanEnterTransition=Get bIsTeleport())]
 }
 Transition {
  path=ItemFinalPoseOverride.AnimGraphNode_StateMachine_1.Teleport SM.AnimStateTransitionNode_5.Transition
  nodes [
   Result(bCanEnterTransition=Get bIsStartPhase1())
   Get bIsTeleport()
  ]
 }
 ItemFullBodyAimOffsets {nodes [
   Input Pose!() -> (Pose, FullBody_AimOffsetAlpha:double, Aim_Yaw:double, Aim_Pitch:double) {name=InPoseFullBodyAimOffsets}
   Input Pose!() -> (Pose) {name=InPosePreFullBodyAimOffsets}
   Output Pose() {layerGroup=Item Override}
  ]}
 ItemFullBodyOverride {nodes [
   Input Pose() {name=InPoseFullBody}
   Output Pose() {layerGroup=Item Override}
  ]}
 ItemIdleOverride {nodes [
   Input Pose() {name=InPose_Idle}
   Output Pose() {layerGroup=Item Override}
  ]}
 ItemJetPackJumpAdditive {nodes [Output Pose() {layerGroup=Item Override}]}
 ItemJetPackStartAdditive {nodes [Output Pose() {layerGroup=Item Override}]}
 ItemJumpLoopAdditive {nodes [
   Input Pose() {name=InPoseItemJumpLoopAdditive}
   Output Pose() {layerGroup=Item Override}
  ]}
 ItemJumpOverride {nodes [
   Input Pose() {name=InPoseJump}
   Output Pose() {layerGroup=Item Override}
  ]}
 ItemJumpUpAdditive {nodes [
   Input Pose() {name=InPoseItemJumpUpAdditive}
   Output Pose() {layerGroup=Item Override}
  ]}
 ItemPreAimOffsets {nodes [
   Input Pose() {name=InPosePreAimOffsets}
   Output Pose() {layerGroup=Item Override}
  ]}
 ItemPreIK {nodes [
   Input Pose() {name=InPose_PreIK}
   Output Pose() {layerGroup=Item Override}
  ]}
 ItemRightHandIK {nodes [
   Input Pose() {name=InPosePreRightHandIK}
   Input Pose() {name=InPosePostRightHandIK}
   Output Pose() {layerGroup=Item Override}
  ]}
 ItemSkeletalControl {nodes [
   Input Pose()
   Output Pose() {layerGroup=Item Override}
  ]}
 ItemSwimJumpFallLoop {nodes [Output Pose() {layerGroup=Item Override}]}
 ItemSwimJumpSurfaceEnd {nodes [Output Pose() {layerGroup=Item Override}]}
 ItemSwimJumpSurfaceLoop {nodes [Output Pose() {layerGroup=Item Override}]}
 ItemUpperBody {nodes [
   Input Pose!() -> (Pose, InputParam:Fort Anim Input Adjusted Aim Structure) {name=InPoseUpperBody}
   Output Pose() {layerGroup=Item Override}
  ]}
 ItemZipLineStartAdditive {nodes [Output Pose() {layerGroup=Item Override}]}
]
variables []
